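# Physics asset for the Body_MoonFlaxSports vehicle body cosmetic, a SportsCar MoonFlax variant with a PhysicsBR suffix.
# It defines a single body on the root bone, with no collision shapes listed for it.
# It has no constraints, so it describes no articulated or ragdoll joints.
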PhysicsAsset /VehicleCosmetics/Bodies/SportsCar_MoonFlax/Meshes/Body_MoonFlaxSports_PhysicsBR
bodies {root []}
constraints []
