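# Physics asset for the MoonFlax sports car body mesh, a vehicle cosmetic.
# A single body on the root bone approximates the whole car with one box shape, offset forward along X and raised above the bone origin.
# It has no constraints and no separate bodies for other bones, so collision is one coarse volume rather than an articulated setup.
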PhysicsAsset /VehicleCosmetics/Bodies/SportsCar_MoonFlax/Meshes/Body_MoonFlaxSports_PhysicsDM
bodies {root [{shapeType=Box, extentX=444, extentY=216, extentZ=116, shapeName=root_box, center=<39.0291 0 91.2142>}]}
constraints []
